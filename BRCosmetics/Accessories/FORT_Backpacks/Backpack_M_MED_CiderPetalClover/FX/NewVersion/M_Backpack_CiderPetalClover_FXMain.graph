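# Unlit, two-sided translucent material for Niagara mesh particles that renders flowing trail FX, blue by default, for the CiderPetalClover backpack.
# Vertex offset blends between tangent-based widening and SplineThickenWithTangentInput, with width shaped by UV masks, ramps and dynamic parameters.
# Emissive sums scrolling base, highlight, line and edge layers, using noise-driven BumpOffset distortion on a shared CommonSpeed clock.
# Opacity combines particle alpha, panning smoke, UV end fades and view-angle fades.
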Material /BRCosmetics/Accessories/FORT_Backpacks/Backpack_M_MED_CiderPetalClover/FX/NewVersion/M_Backpack_CiderPetalClover_FXMain
graph {
 properties {
  bAutomaticallySetUsageInEditor=false
  bUsedWithNiagaraMeshParticles=true
  blendMode=BLEND_Translucent
  shadingModel=MSM_Unlit
  twoSided=true
  usedShadingModels=Unlit
 }
 nodes [
  Constant2Vector() {g=0.2, r=0.5}
  n1 = ComponentMask(None=TextureCoordinate()) {g=true}
  n2 = ComponentMask(None=TextureCoordinate()) {r=true}
  n3 = Power(Base=Add(A=OneMinus(None=@n2)) {constB=0.1}) {constExponent=0.8}
  n4 = Multiply(
   A=Multiply(A=Subtract(A=@n1) {constB=0.5}, B=MaterialFunctionCall() {materialFunction=&TangentBasis}.Y V)
   B=Multiply(A=LinearInterpolate(Alpha=@n3))
  )
  n5 = TextureCoordinate()
  n6 = LinearInterpolate(B=LinearInterpolate(Alpha=ComponentMask(None=@n5) {r=true}) {constA=0.2, constB=1.4}, Alpha=VertexColor().R) {constA=1, constB=0.7}
  n7 = MaterialFunctionCall() {materialFunction=&TangentBasis}
  n8 = LinearInterpolate(Alpha=ScalarParameter() {parameterName=UseSplineAngleMask}) {constA=1, constB=-1}
  n9 = Multiply(A=@n7.X U, B=@n8)
  n10 = Multiply(A=@n7.Y V, B=@n8)
  n11 = TextureCoordinate()
  n12 = TextureCoordinate()
  n13 = Power(
   Base=OneMinus(None=ComponentMask(None=@n12) {r=true})
   Exp=ScalarParameter() {defaultValue=1, group=CustomSpline, parameterName=CustomSplineWidthPower}
  )
  n14 = ScalarParameter() {defaultValue=1, group=CustomSpline, parameterName=CustomSplineEdge}
  n15 = LinearInterpolate(
   A=ColorRamp!(None=ComponentMask(None=@n11) {r=true}) {colorCurve="<self>.ColorCurve"}
   B=Saturate(None=Multiply(A=@n13, B=@n14))
   Alpha=ScalarParameter() {group=CustomSpline, parameterName=UseCustomWidthForSpline}
  )
  n16 = SphereMask(A=TextureCoordinate() {vTiling=0}, B=Constant2Vector() {r=0.45}) {attenuationRadius=0.3, hardnessPercent=0}
  n17 = Sine(None=ComponentMask(None=TextureCoordinate() {vTiling=0}) {r=true}) {period=2}
  n18 = Multiply(A=LinearInterpolate(Alpha=OneMinus(None=@n16)) {constA=0.5}, B=LinearInterpolate(Alpha=@n17) {constA=0.3})
  n19 = ScalarParameter() {parameterName=UseSplineAngleMask}
  n20 = Multiply(
   A=Multiply(A=ScalarParameter() {defaultValue=2, parameterName=Width}, B=LinearInterpolate(A=@n15, B=@n18, Alpha=@n19) {constA=1})
   B=DynamicParameter() {defaultValue=<rgba 10 1 1 1>, paramNames [Width,Param2,Param3,Param4]}.Width
  )
  n21 = StaticBool()
  LinearInterpolate(
   A=Multiply(A=Multiply(A=@n4, B=@n6), B=DynamicParameter() {defaultValue=<rgba 10 1 1 1>, paramNames [Width,Param2,Param3,Param4]}.Width)
   B=Multiply(A=MaterialFunctionCall!(
     TangentU_WS=@n9
     TangentV_WS=@n10
     WidthBase=@n20
     WidthTip=@n20
     WorldPosition
     UVs for Projection
     UVs for Thickness
     UVs for texturing
     Expand U or V UV Channel=@n21
     DeriveNormalZ
     AdditionalNormal
     AngleCorrectedNormals
     FlattenPixelNormal
     UVs For NormalShading
    ) {materialFunction=/Game/Effects/Niagara/Modules/FollowingDeform/Material/SplineThickenWithTangentInput}.WorldPositionOffset)
   Alpha=ScalarParameter() {parameterName=UseSplineThicken}
  )
  n22 = TextureCoordinate() {uTiling=1.5, vTiling=0.3}
  n23 = NamedRerouteUsage()
  n24 = Power(Base=TextureSample(UVs=Panner(Coordinate=@n22, Time=@n23) {speedX=-0.2}) {texture=/Game/Effects/Niagara/MasterMaterials/SplineThickenTrail/Texture/T_Pickaxe_SoapPocket_NoiseGlow}.R) {constExponent=1.5}
  n25 = LinearInterpolate(
   A=Multiply(
    A=VectorParameter() {parameterName=BottomColorDark, defaultValue=<rgba 0 0.0104 0.1 1>}.RGB
    B=LinearInterpolate(Alpha=@n24) {constA=1, constB=10}
   )
   B=VectorParameter() {parameterName=BottomColorLight, defaultValue=<rgba 0 0.052 0.5 1>}.RGB
   Alpha=TextureSample(UVs=TextureCoordinate()) {samplerType=SAMPLERTYPE_Grayscale, texture=&LowResBlurredNoise}.R
  )
  n26 = Multiply(
   A=TextureCoordinate()
   B=ComponentMask(None=VectorParameter() {group=BaseBlue, parameterName=BaseBlueUV, defaultValue=<rgba 1 0.45 0 1>}.RGB) {g=true, r=true}
  )
  n27 = VectorParameter() {group=BaseBlue, parameterName=BaseBlueOffset, defaultValue=<rgba 0 0.05 0 1>}
  n28 = Panner(Coordinate=Add(A=Add(A=@n26) {constB=0}, B=ComponentMask(None=@n27.RGB) {g=true, r=true}), Time=NamedRerouteUsage()) {speedX=-0.3}
  n29 = Multiply(
   A=VectorParameter() {group=BaseBlue, parameterName=BaseBlueColor, defaultValue=<rgba 0 3.5837 10 1>}.RGB
   B=Power(Base=TextureSample(UVs=@n28) {texture=/Game/Gadgets/Assets/VinderTech_GliderChute/Glider_Emperor/FX/Texture/T_Emperor_Silk_01}.R) {constExponent=1}
  )
  n30 = Multiply(
   A=TextureCoordinate()
   B=ComponentMask(None=VectorParameter() {group=HighLight, parameterName=HighLightUV, defaultValue=<rgba 0.7 0.7 0 0>}.RGB) {g=true, r=true}
  )
  n31 = VectorParameter() {group=HighLight, parameterName=HighLightOffset, defaultValue=<rgba 0.35 -0.1 0 0>}
  n32 = Panner(
   Coordinate=Add(A=Add(A=@n30) {constB=0}, B=ComponentMask(None=@n31.RGB) {g=true, r=true})
   Time=NamedRerouteUsage()
   Speed=ComponentMask(None=VectorParameter() {group=HighLight, parameterName=HighLightSpeed}.RGB) {g=true, r=true}
  )
  n33 = Multiply(
   A=VectorParameter() {group=HighLight, parameterName=HighLightColor, defaultValue=<rgba 0.5853 0.8837 1 1>}.RGB
   B=Power(Base=TextureSample(UVs=@n32) {samplerType=SAMPLERTYPE_LinearColor, texture=/Game/Effects/Fort_Effects/Textures/T_CubeNinja_Trail}.R) {constExponent=1.5}
  )
  n34 = Panner(Coordinate=TextureCoordinate() {vTiling=0.5}, Time=NamedRerouteUsage()) {speedX=-0.5}
  n35 = LinearInterpolate(Alpha=Power(Base=TextureSample(UVs=@n34) {samplerType=SAMPLERTYPE_Grayscale, texture=&LowResBlurredNoise}.R) {constExponent=1.2}) {constA=0.3, constB=2.6}
  n36 = Multiply(
   A=TextureCoordinate()
   B=ComponentMask(None=VectorParameter() {group=LineColor, parameterName=LineColorUV, defaultValue=<rgba 0.7 0.8 0 0>}.RGB) {g=true, r=true}
  )
  n37 = TextureSample(UVs=Panner(Coordinate=TextureCoordinate() {uTiling=1.2}, Time=NamedRerouteUsage())) {texture=&T_Pickaxe_SoapPocket_NoiseGlow}
  n38 = TextureSample(UVs=Panner(Coordinate=BumpOffset(Coordinate=@n36, Height=@n37.G) {heightRatio=1}, Time=NamedRerouteUsage()) {speedX=-0.2}) {texture=/BRCosmetics/Weapons/FORT_Melee/Pickaxe_Gimmick_Female/FX/T_Pickaxe_Gimmick_Female_Trail_01}
  n39 = Add(
   A=Multiply(
    A=VectorParameter() {group=LineColor, parameterName=LineColor, defaultValue=<rgba 0 1.6828 4 1>}.RGB
    B=Power(Base=@n38.B) {constExponent=1.5}
   )
   B=Add(A=Multiply(A=@n33, B=@n35), B=Add(A=@n29, B=@n25))
  )
  n40 = VectorParameter() {group=EdgeColor, parameterName=EdgeColorUV, defaultValue=<rgba 0.3 0.8 0 1>}
  n41 = Add(
   A=Multiply(A=TextureCoordinate(), B=ComponentMask(None=@n40.RGB) {g=true, r=true})
   B=ComponentMask(None=VectorParameter() {group=EdgeColor, parameterName=EdgeColorOffset, defaultValue=<rgba 0.5 0.2 0 0>}.RGB) {g=true, r=true}
  )
  n42 = TextureSample(UVs=Panner(Coordinate=TextureCoordinate() {uTiling=1.2}, Time=NamedRerouteUsage()) {speedX=-0.35}) {texture=&T_Pickaxe_SoapPocket_NoiseGlow}
  n43 = TextureSample(UVs=Panner(Coordinate=BumpOffset(Coordinate=@n41, Height=@n42.R) {heightRatio=0.3}, Time=NamedRerouteUsage())) {texture=&T_Pickaxe_Gimmick_Female_Trail_01}
  n44 = Multiply(
   A=VectorParameter() {group=EdgeColor, parameterName=EdgeColor, defaultValue=<rgba 0 3.6491 10 1>}.RGB
   B=Power(Base=Multiply(A=@n43.R, B=@n43.R)) {constExponent=1}
  )
  MaterialFunctionCall(Emissive=Multiply(A=ParticleColor().RGB, B=Add(A=@n44, B=@n39))) {materialFunction=&MF_FX_EmissiveAdjustments}
  n45 = MaterialFunctionCall("FadeSensitivity (optional)"=ScalarParameter() {parameterName="FadeSensitivity (optional) (S)"}) {materialFunction=&FadeBasedOnViewAngle}
  n46 = Sine(None=ComponentMask(None=TextureCoordinate()) {r=true}) {period=2}
  n47 = Saturate(None=Multiply(A=Power(Base=@n46) {constExponent=2.2}) {constB=2})
  n48 = ScalarParameter() {parameterName=UseSplineAngleMask}
  n49 = TextureCoordinate()
  n50 = OneMinus(None=ComponentMask(None=@n49) {g=true})
  n51 = Constant() {r=2}
  n52 = ComponentMask(None=@n49) {r=true}
  n53 = Constant() {r=1.8}
  n54 = Multiply(A=Multiply(A=Power(Base=@n50, Exp=@n51) {constExponent=1.2}, B=Power(Base=@n52, Exp=@n53) {constExponent=1.2})) {constB=1000}
  n55 = Multiply(A=MaterialFunctionCall("FadeSensitivity (optional)"=ScalarParameter() {defaultValue=2, parameterName=AngleFade}) {materialFunction=&FadeBasedOnViewAngle}.Result) {constB=5}
  n56 = ParticleColor()
  n57 = Panner(Coordinate=Add(A=TextureCoordinate(), B=Constant2Vector()), Time=NamedRerouteUsage()) {speedX=-0.3}
  n58 = Power(Base=OneMinus(None=TextureSample(UVs=@n57) {samplerType=SAMPLERTYPE_Grayscale, texture=/Game/Effects/Fort_Effects/Textures/Smoke/T_Smoke_Trail_Soft_NoClamp}.R)) {constExponent=3}
  n59 = Saturate(None=Multiply(A=OneMinus(None=@n58)))
  n60 = Multiply(A=Multiply(A=Multiply(A=@n56.A, B=@n59), B=Saturate(None=@n55)), B=Saturate(None=@n54))
  n61 = OneMinus(None=@n52)
  n62 = ScalarParameter() {defaultValue=1.2, parameterName=TailFadePower}
  n63 = Saturate(None=Multiply(A=Power(Base=@n61, Exp=@n62) {constExponent=1}) {constB=20})
  Multiply(
   A=Multiply(A=Multiply(A=@n60, B=@n63), B=LinearInterpolate(B=@n47, Alpha=@n48) {constA=1})
   B=Saturate(None=Multiply(A=@n45.Result) {constB=5})
  )
  NamedRerouteDeclaration(None=Multiply(A=Time(), B=ScalarParameter() {defaultValue=1, parameterName=CommonSpeed})) {name=Speed, nodeColor=<rgba 0 0.0859 1 1>}
  OneMinus(None=Power(Base=@n17) {constExponent=5})
  Round(None=@n54)
 ]
}
