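# Niagara system for the CiderPetalClover backpack cosmetic, drawing a glowing layered shape with sparks.
# MainShape, Head, Spike and Angle burst-spawn persistent meshes, with a light on MainShape and six partition-placed Glow sprites, positioned using a shared System.FXOffset.
# Sparks spawn continuously, sampling positions from the User.Static Mesh parameter, with drag and speed-based stretching.
# User.FXColorSwitch blends colors from mostly blue to orange-gold, while front-end modifiers adjust some brightness and alpha.
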NiagaraSystem /BRCosmetics/Accessories/FORT_Backpacks/Backpack_M_MED_CiderPetalClover/FX/NewVersion/NS_Backpack_CiderPetalClover_FXNew
dependencies {
 usedDataInterfaces [
  &NiagaraDataInterfaceArrayFloat3
  &NiagaraDataInterfaceArrayFloat
  &NiagaraDataInterfaceStaticMesh
  &NiagaraDataInterfaceCurve
 ]
 usedDynamicInputs []
 usedModules [
  &self:SystemSpawnScript.SystemScriptSource.SystemScriptGraph.NiagaraNodeAssignment_0.SetVariables_#1
  &SystemState
 ]
}
emitters {
 Angle {
  emitterUpdateScript [
   &EmitterState(Life Cycle Mode=::NewEnumerator1, Loop Behavior=::NewEnumerator1, Loop Duration Mode=::NewEnumerator1, UseLoopDelay=false)
   &SpawnBurst_Instantaneous(Use Spawn Probability=false, UseLoopCountLimit=false)
   /SetVariables(Emitter.BaseBlueColor:LinearColor=&LerpLinearColors(StartColor=<rgba 0 3.5837 10 1>, EndColor=<rgba 2 1 0 1>, LerpFactor=$User.FXColorSwitch)) {isSetParameters=true, module=SetVariables_#2}
   /SetVariables(Emitter.HighLightColor:LinearColor=&LerpLinearColors(StartColor=<rgba 0.0361 1.0241 3 1>, EndColor=<rgba 2 1.4146 0.2437 1>, LerpFactor=$User.FXColorSwitch)) {isSetParameters=true, module=SetVariables_#3}
   /SetVariables(Emitter.LineColor:LinearColor=&LerpLinearColors(StartColor=<rgba 0 0 0 1>, EndColor=<rgba 0 0 0 1>, LerpFactor=$User.FXColorSwitch)) {isSetParameters=true, module=SetVariables_#4}
   /SetVariables(Emitter.EdgeColor:LinearColor=&LerpLinearColors(StartColor=<rgba 0 0 0 1>, EndColor=<rgba 0 0 0 1>, LerpFactor=$User.FXColorSwitch)) {isSetParameters=true, module=SetVariables_#5}
   /SetVariables(Emitter.BottomColorDark:LinearColor=&LerpLinearColors(StartColor=<rgba 0 0.0104 0.1 1>, EndColor=<rgba 0.1 0.0292 0 1>, LerpFactor=$User.FXColorSwitch)) {isSetParameters=true, module=SetVariables_#6}
   /SetVariables(Emitter.BottomColorLight:LinearColor=&LerpLinearColors(StartColor=<rgba 0 0.0208 0.2 1>, EndColor=<rgba 0.3 0.1 0 1>, LerpFactor=$User.FXColorSwitch)) {isSetParameters=true, module=SetVariables_#7}
  ]
  particleSpawnScript [
   &InitializeParticle(
    Color=&LerpLinearColors(
     StartColor=<rgba 3 3 3 1.3>
     EndColor=<rgba 3 3 3 1>
     LerpFactor=&FN_FrontEndFloatModifier(Gameplay=0, Front End=1)
    )
    Mesh Scale=&Multiply_VectorByFloat(Vector=<1.5 1 0.9>, Float=0.78)
    Mesh Scale Mode=::NewEnumerator4
    Position Offset=&AddVectorToPosition(Position=<-11 0 8>, Vector=$System.FXOffset)
    UsePositionOffset=true
    Write Material Random=true
   )
   &DynamicMaterialParameters(
    Index 0 Param 1=15
    Module.Index0Float0Write=true
    Module.Index0Float1Write=true
    Module.Index0Float2Write=true
    Module.Index0Float3Write=true
   )
   &CameraOffset()
   &InitialMeshOrientation(Mesh Orientation Mode=::NewEnumerator7, Rotation=<0 0.02 0>, Use Rotation Vector=true)
  ]
  particleUpdateScript [&ParticleState(Kill Particles When Lifetime Has Elapsed=false)]
  rendererClasses [&NiagaraMeshRendererProperties]
  renderers [{rendererClass=&NiagaraMeshRendererProperties, rendererIndex=0}]
 }
 Glow {
  emitterUpdateScript [
   &EmitterState(Life Cycle Mode=::NewEnumerator1, Loop Behavior=::NewEnumerator1, Loop Duration Mode=::NewEnumerator1, UseLoopDelay=false)
   &SpawnBurst_Instantaneous(Spawn Count=6, Use Spawn Probability=false, UseLoopCountLimit=false)
  ]
  particleSpawnScript [
   &PartitionParticles(Number Of Partitions=6)
   &InitializeParticle(
    Color Mode=::NewEnumerator0
    Position Offset=&AddVectorToPosition(
     Position=&SelectVectorFromArray(
      Direct Array Index=$Particles.PartitionParticles.Partition
      Vector Selection Array=DataInterface() {propertyValues {FloatData [<0 0 12.5>, <-10 0 12.5>, <-11 0 5>, <-6 0 -3>, <-1.5 0 -9>, <2 0 -13>]}}
     )
     Vector=$System.FXOffset
    )
    Sprite Size Mode=::NewEnumerator3
    Uniform Sprite Size=&SelectFloatFromArray(
     Direct Array Index=$Particles.PartitionParticles.Partition
     Float Selection Array=DataInterface() {propertyValues {FloatData [30,30,27,22,17,12]}}
    )
    UsePositionOffset=true
    Write Material Random=true
   )
   &DynamicMaterialParameters(
    Index 0 Param 1=3
    Module.Index0Float0Write=true
    Module.Index0Float1Write=true
    Module.Index0Float2Write=true
    Module.Index0Float3Write=true
   )
   &CameraOffset()
  ]
  particleUpdateScript [
   &ParticleState(Kill Particles When Lifetime Has Elapsed=false)
   &Color(Color=&MakeLinearColorFromVectorAndFloat(
     "Vector (RGB)"=&MakeVectorFromLinearColorRGB(Color=&LerpLinearColors(StartColor=<rgba 0 0.2557 1 2>, EndColor=<rgba 0.4 0.2195 0.039 2>, LerpFactor=$User.FXColorSwitch))
     "Float (Alpha)"=&FN_FrontEndFloatModifier(Gameplay=0.8, Front End=0.3)
    ))
  ]
  rendererClasses [&NiagaraSpriteRendererProperties]
  renderers [{rendererClass=&NiagaraSpriteRendererProperties}]
 }
 Head {
  emitterUpdateScript [
   &EmitterState(Life Cycle Mode=::NewEnumerator1, Loop Behavior=::NewEnumerator1, Loop Duration Mode=::NewEnumerator1, UseLoopDelay=false)
   &SpawnBurst_Instantaneous(Use Spawn Probability=false, UseLoopCountLimit=false)
   /SetVariables(Emitter.BaseBlueColor:LinearColor=&LerpLinearColors(StartColor=<rgba 0 0.1268 3 1>, EndColor=<rgba 3 1.5 0 1>, LerpFactor=$User.FXColorSwitch)) {isSetParameters=true, module=SetVariables_#2}
   /SetVariables(Emitter.HighLightColor:LinearColor=&LerpLinearColors(StartColor=<rgba 0.5853 0.8837 1 1>, EndColor=<rgba 2 1.4146 0.2437 1>, LerpFactor=$User.FXColorSwitch)) {isSetParameters=true, module=SetVariables_#3}
   /SetVariables(Emitter.LineColor:LinearColor=&LerpLinearColors(StartColor=<rgba 0 0 0 1>, EndColor=<rgba 0 0 0 1>, LerpFactor=$User.FXColorSwitch)) {isSetParameters=true, module=SetVariables_#4}
   /SetVariables(Emitter.EdgeColor:LinearColor=&LerpLinearColors(StartColor=<rgba 0 0.3649 1 1>, EndColor=<rgba 1 0.5 0 1>, LerpFactor=$User.FXColorSwitch)) {isSetParameters=true, module=SetVariables_#5}
   /SetVariables(Emitter.BottomColorDark:LinearColor=&LerpLinearColors(StartColor=<rgba 0 0.0104 0.1 1>, EndColor=<rgba 0.1 0.0292 0 1>, LerpFactor=$User.FXColorSwitch)) {isSetParameters=true, module=SetVariables_#6}
   /SetVariables(Emitter.BottomColorLight:LinearColor=&LerpLinearColors(StartColor=<rgba 0 0.052 0.5 1>, EndColor=<rgba 0.3 0.1 0 1>, LerpFactor=$User.FXColorSwitch)) {isSetParameters=true, module=SetVariables_#7}
  ]
  particleSpawnScript [
   &InitializeParticle(
    Color=&LerpLinearColors(
     StartColor=<rgba 5 5 5 1>
     EndColor=<rgba 3 3 3 1>
     LerpFactor=&FN_FrontEndFloatModifier(Gameplay=0, Front End=1)
    )
    Mesh Scale=&Multiply_VectorByFloat(Vector=<1.5 1 0.9>, Float=0.78)
    Mesh Scale Mode=::NewEnumerator4
    Position Offset=&AddVectorToPosition(Position=<-6.5 0 13>, Vector=$System.FXOffset)
    UsePositionOffset=true
    Write Material Random=true
   )
   &DynamicMaterialParameters(
    Index 0 Param 1=13
    Module.Index0Float0Write=true
    Module.Index0Float1Write=true
    Module.Index0Float2Write=true
    Module.Index0Float3Write=true
   )
   &CameraOffset(Camera Offset Amount=5)
   &InitialMeshOrientation(Mesh Orientation Mode=::NewEnumerator7, Rotation=<0 0.02 0>, Use Rotation Vector=true)
  ]
  particleUpdateScript [&ParticleState(Kill Particles When Lifetime Has Elapsed=false)]
  rendererClasses [&NiagaraMeshRendererProperties]
  renderers [{rendererClass=&NiagaraMeshRendererProperties, rendererIndex=0}]
 }
 MainShape {
  emitterUpdateScript [
   &EmitterState(Life Cycle Mode=::NewEnumerator1, Loop Behavior=::NewEnumerator1, Loop Duration Mode=::NewEnumerator1, UseLoopDelay=false)
   &SpawnBurst_Instantaneous(Use Spawn Probability=false, UseLoopCountLimit=false)
   /SetVariables(Emitter.BaseBlueColor:LinearColor=&LerpLinearColors(StartColor=<rgba 0 3.5837 10 1>, EndColor=<rgba 8 5.3333 0 1>, LerpFactor=$User.FXColorSwitch)) {isSetParameters=true}
   /SetVariables(Emitter.HighLightColor:LinearColor=&LerpLinearColors(StartColor=<rgba 0.5853 0.8837 1 1>, EndColor=<rgba 1 0.6797 0.3594 1>, LerpFactor=$User.FXColorSwitch)) {isSetParameters=true}
   /SetVariables(Emitter.LineColor:LinearColor=&LerpLinearColors(StartColor=<rgba 0 1.6828 4 1>, EndColor=<rgba 4 2.6667 0 1>, LerpFactor=$User.FXColorSwitch)) {isSetParameters=true}
   /SetVariables(Emitter.EdgeColor:LinearColor=&LerpLinearColors(StartColor=<rgba 0 3.6491 10 1>, EndColor=<rgba 10 6.6667 0 1>, LerpFactor=$User.FXColorSwitch)) {isSetParameters=true}
   /SetVariables(Emitter.BottomColorDark:LinearColor=&LerpLinearColors(StartColor=<rgba 0 0.0104 0.1 1>, EndColor=<rgba 0.1 0.0292 0 1>, LerpFactor=$User.FXColorSwitch)) {isSetParameters=true}
   /SetVariables(Emitter.BottomColorLight:LinearColor=&LerpLinearColors(StartColor=<rgba 0 0.052 0.5 1>, EndColor=<rgba 0.5 0.1667 0 1>, LerpFactor=$User.FXColorSwitch)) {isSetParameters=true}
  ]
  particleSpawnScript [
   &InitializeParticle(Mesh Scale=<1 0.1 1>, Mesh Scale Mode=::NewEnumerator4, Position Offset=$System.FXOffset, UsePositionOffset=true, Write Material Random=true)
   &CameraOffset()
   &DynamicMaterialParameters(
    Index 0 Param 1=13
    Module.Index0Float0Write=true
    Module.Index0Float1Write=true
    Module.Index0Float2Write=true
    Module.Index0Float3Write=true
   )
  ]
  particleUpdateScript [
   &ParticleState(Kill Particles When Lifetime Has Elapsed=false)
   &ScaleColor(ScaleA=true, ScaleRGB=true)
   &Light_Attributes(
    Color=1
    Color Mode=0
    Falloff Exponent=2
    Light Color:LinearColor=&LerpLinearColors(StartColor=<rgba 0 5.0266 20 1>, EndColor=<rgba 10 5 0 1>, LerpFactor=$User.FXColorSwitch)
    Light Exponent=2.5
    Light Position Offset:Vector3f=<0 -5 0>
    Light Radius=35
    Position=1
    Radius=2
    UsePositionOffset=true
   )
  ]
  rendererClasses [&NiagaraMeshRendererProperties, &NiagaraLightRendererProperties]
  renderers [
   {rendererClass=&NiagaraMeshRendererProperties, rendererIndex=0}
   {rendererClass=&NiagaraLightRendererProperties, rendererIndex=1}
  ]
 }
 Sparks {
  emitterUpdateScript [
   &EmitterState(Inactive Response=::NewEnumerator1, Life Cycle Mode=::NewEnumerator1, Loop Duration=4.6, UseLoopDelay=false)
   &SpawnRate(SpawnRate=40, Use Spawn Probability=false)
  ]
  particleSpawnScript [
   &InitializeParticle(
    Lifetime Max=0.7
    Lifetime Min=0.5
    Lifetime Mode=::NewEnumerator1
    Ribbon Width=40
    Ribbon Width Mode=::NewEnumerator1
    Sprite Rotation Mode=::NewEnumerator2
    Sprite Size Max=<0.7 5>
    Sprite Size Min=<0.5 2>
    Sprite Size Mode=::NewEnumerator2
    UsePositionOffset=true
    Write Material Random=true
   )
   &AddVelocity(Velocity=<-40 0 5>) {module=AddVelocity001}
   &CameraOffset(Camera Offset Amount=10)
  ]
  particleUpdateScript [
   /Niagara/Modules/Spawn/Location/StaticMeshLocation(
    MaskSpawnGroup=false
    Mesh Sampling Type:ENiagara_StaticSamplingMode=::NewEnumerator2
    OffsetAlongNormal=false
    OverrideRandomSeed=false
    Position Sampling:ENiagaraSkelMeshPositionSamplingMode=::NewEnumerator2
    SamplePosOffset=true
    Sampled Position Offset:Vector3f=&AddVectorToPosition(
     Position=Linked() {linkedVariable {name=System.FXOffset, type=Vector3f}}
     Vector=&Multiply_VectorByFloat(Vector=<0 0 0>, Float=1)
    )
    Sampled Position Offset Coordinate Space:ENiagaraSkelMeshTransforms=::NewEnumerator2
    Static Mesh=DataInterface() {propertyValues {
      DefaultMesh=null
      InstanceIndex=-1
      LODIndex=0
      PreviewMesh=&/SM_Backpack_CiderPetalClover_Base
      SoftSourceActor=null
      SourceMode=MeshParameterBinding
      bAllowSamplingFromStreamingLODs=false
      bCaptureTransformsPerFrame=true
      bUsePhysicsBodyVelocity=false
      FilteredSockets []
      LODIndexUserParameter {parameter {name=None, typeDefHandle {registeredTypeIndex=108}, varData []}}
      MeshParameterBinding {parameter {name=User.Static Mesh, typeDefHandle {registeredTypeIndex=1}, varData []}}
      SectionFilter {allowedMaterialSlots []}
     }}
   )
   /SetVariables(Particles.SizeAll=1) {isSetParameters=true}
   &ParticleState()
   &ScaleColor(
    Scale RGB=&MakeVectorFromLinearColorRGB(Color=&LerpLinearColors(StartColor=<rgba 2.7735 8.5157 20 1>, EndColor=<rgba 20 11.3868 2.7735 1>, LerpFactor=$User.FXColorSwitch))
    ScaleA=true
    ScaleRGB=true
   )
   &ScaleSpriteSize(Scale Factor=&Vector2DFromFloat(Value=&FloatFromCurve(
      FloatCurve=DataInterface() {propertyValues {Curve {keys [
          <at 0 0 RCIM_Cubic>
          {interpMode=RCIM_Cubic, leaveTangent=-3.0296, tangentMode=RCTM_Break, time=0.2, value=1.2}
          {arriveTangent=-1.8107, interpMode=RCIM_Cubic, leaveTangent=-1.8107, tangentMode=RCTM_User, time=1, value=0}
         ]}}}
      Scale Curve=&Multiply_Float(A=0.4, B=$Particles.SizeAll)
     ))) {module=ScaleSpriteSize001}
   /SetVariables(Particles.Glow03_Color=&MakeLinearColorFromVectorAndFloat(
     "Vector (RGB)"=&MakeVectorFromLinearColorRGB(Color=&MultiplyLinearColor(
       Linear Color=&LerpLinearColors(StartColor=<rgba 0.0112 0.176 1 1>, EndColor=<rgba 1 0.5056 0.0112 1>, LerpFactor=$User.FXColorSwitch)
       Scale Factor=&MakeVector4_FromFloat(Float=1)
      ))
     "Float (Alpha)"=&Multiply_Float(A=0.3, B=&FN_FrontEndFloatModifier(Gameplay=1, Front End=0.5))
    )) {isSetParameters=true}
   /SetVariables(Particles.Glow03_Size=&Multiply_Vector2D_ByFloat(
     Float=&FloatFromCurve(
      FloatCurve=DataInterface() {propertyValues {Curve {keys [
          <at 0 0 RCIM_Cubic>
          <at 0.2 1.3 RCIM_Cubic>
          {arriveTangent=-0.1423, interpMode=RCIM_Cubic, leaveTangent=-0.1423, tangentMode=RCTM_User, time=1, value=0}
         ]}}}
      Scale Curve=8
     )
     Vector2D=&Multiply_Vector2D_ByFloat(Float=$Particles.SizeAll, Vector2D=<0.2 0.6>)
    )) {isSetParameters=true}
   &DynamicMaterialParameters(
    Module.Param0WriteEnabled=true
    Module.Param0WriteEnabled1=true
    Module.Param1WriteEnabled=true
    Module.Param1WriteEnabled1=true
    Module.Param2WriteEnabled=true
    Module.Param2WriteEnabled1=true
    Module.Param3WriteEnabled=true
    Module.Param3WriteEnabled1=true
    Write Parameter Index 1=true
   )
   &Drag(Use Linear Drag=true, Use Rotational Drag=true)
   &GravityForce(Gravity=<0 0 15>) {enabled=false}
   &SolveForcesAndVelocity(Clamp Velocity=false, Limit Acceleration=false)
   &ScaleSpriteSizeBySpeed(Max Scale Factor=<2 10>)
   /SetVariables(Particles.Position_2:NiagaraPosition=&AddVectorToPosition(Position=$Particles.StaticMeshLocation.SampledPosition, Vector=Linked() {linkedVariable {name=Particles.Position, type=NiagaraPosition}})) {isSetParameters=true}
  ]
  rendererClasses [&NiagaraSpriteRendererProperties]
  renderers [
   {rendererClass=&NiagaraSpriteRendererProperties}
   {rendererClass=&NiagaraSpriteRendererProperties, rendererIndex=1}
  ]
 }
 Spike {
  emitterUpdateScript [
   &EmitterState(Life Cycle Mode=::NewEnumerator1, Loop Behavior=::NewEnumerator1, Loop Duration Mode=::NewEnumerator1, UseLoopDelay=false)
   &SpawnBurst_Instantaneous(Use Spawn Probability=false, UseLoopCountLimit=false)
   /SetVariables(Emitter.BaseBlueColor:LinearColor=&LerpLinearColors(StartColor=<rgba 5 5 5 1>, EndColor=<rgba 1.536 3 1.4063 1>, LerpFactor=$User.FXColorSwitch)) {isSetParameters=true}
   /SetVariables(Emitter.HighLightColor:LinearColor=&LerpLinearColors(StartColor=<rgba 0.5853 0.8837 1 1>, EndColor=<rgba 1 0.7926 0.5853 1>, LerpFactor=$User.FXColorSwitch)) {isSetParameters=true}
   /SetVariables(Emitter.LineColor:LinearColor=&LerpLinearColors(StartColor=<rgba 0 0 0 1>, EndColor=<rgba 0 0 0 1>, LerpFactor=$User.FXColorSwitch)) {isSetParameters=true}
   /SetVariables(Emitter.EdgeColor:LinearColor=&LerpLinearColors(StartColor=<rgba 0 0.3649 1 1>, EndColor=<rgba 1 0.5 0 1>, LerpFactor=$User.FXColorSwitch)) {isSetParameters=true}
   /SetVariables(Emitter.BottomColorDark:LinearColor=&LerpLinearColors(StartColor=<rgba 0 0.0104 0.1 1>, EndColor=<rgba 0.1 0.05 0 1>, LerpFactor=$User.FXColorSwitch)) {isSetParameters=true}
   /SetVariables(Emitter.BottomColorLight:LinearColor=&LerpLinearColors(StartColor=<rgba 0 0.052 0.5 1>, EndColor=<rgba 0.1 0.05 0 1>, LerpFactor=$User.FXColorSwitch)) {isSetParameters=true}
  ]
  particleSpawnScript [
   &InitializeParticle(
    Mesh Scale=&Multiply_VectorByFloat(Vector=<0.9 1 1>, Float=0.39)
    Mesh Scale Mode=::NewEnumerator4
    Position Offset=&AddVectorToPosition(Position=<0 0 13>, Vector=$System.FXOffset)
    UsePositionOffset=true
    Write Material Random=true
   )
   &DynamicMaterialParameters(
    Index 0 Param 1=8
    Module.Index0Float0Write=true
    Module.Index0Float1Write=true
    Module.Index0Float2Write=true
    Module.Index0Float3Write=true
   )
   &CameraOffset(Camera Offset Amount=5)
   &InitialMeshOrientation(Mesh Orientation Mode=::NewEnumerator7, Rotation=<0 0.53 0>, Use Rotation Vector=true)
  ]
  particleUpdateScript [
   &ParticleState(Kill Particles When Lifetime Has Elapsed=false)
   &ScaleColor(
    Scale RGB=&MakeVectorFromLinearColorRGB(Color=&LerpLinearColors(
      StartColor=&LerpLinearColors(
       StartColor=<rgba 0 3.3333 5 1>
       EndColor=<rgba 2.0589 4.1266 5 1>
       LerpFactor=&FN_FrontEndFloatModifier(Gameplay=0, Front End=1)
      )
      EndColor=<rgba 3 1.5 0 1>
      LerpFactor=$User.FXColorSwitch
     ))
    ScaleA=true
    ScaleRGB=true
   )
  ]
  rendererClasses [&NiagaraMeshRendererProperties]
  renderers [{rendererClass=&NiagaraMeshRendererProperties, rendererIndex=0}]
 }
}
summary {
 systemName=NS_Backpack_CiderPetalClover_FXNew
 emitters [
  {emitterName=MainShape, rendererClasses [&NiagaraMeshRendererProperties, &NiagaraLightRendererProperties]}
  {emitterName=Head, rendererClasses [&NiagaraMeshRendererProperties]}
  {emitterName=Spike, rendererClasses [&NiagaraMeshRendererProperties]}
  {emitterName=Glow, rendererClasses [&NiagaraSpriteRendererProperties]}
  {emitterName=Sparks, rendererClasses [&NiagaraSpriteRendererProperties]}
  {emitterName=Angle, rendererClasses [&NiagaraMeshRendererProperties]}
 ]
 userVariables [
  User.bIsFrontEnd {type=NiagaraBool}
  User.bIsFrontEndPreview {type=NiagaraBool}
  User.FXColorSwitch {type=NiagaraFloat, defaultValue=0, description=""}
  User.Static Mesh {
   type=StaticMesh
   description=""
   defaultValue {
    struct=NiagaraExt_VariableValue_Object
    value {
     object=&SM_Backpack_CiderPetalClover_Base
     objectClass=&StaticMesh
    }
   }
  }
 ]
}
